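; Slurp geyser Blueprint in the bug biome that applies a slurp effect to characters it launches.
; When a character is launched, it gets that character's Fort ability system component and, if valid, applies the goo gameplay effect class with itself as source and causer.
; It then adds the looping apply-slurp gameplay cue to the character.
; The construction script only calls the parent, and no launch logic appears in this Blueprint.
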(event EventLaunchedCharacter
  (params (Character :type Character%20Object%20Reference))
  (bind _returnvalue (Ability|GetFortAbilitySystemComponent Character))
  (Utilities|IsValid _returnvalue
    (:Is%20Valid
      (GameplayEffects|FortApplyGameplayEffectToTarget :self (Ability|GetFortAbilitySystemComponent self) :GameplayEffectClass (Variables|Default|GetGooEffectClass) :TargetOfEffect _returnvalue :OptionalSourceObject self :OptionalEffectCauser self :OptionalDamageSource self :Level 0.0)
      (GameplayCue|AddGameplayCueOnActor(Looping) Character (Variables|Default|GetGC_ApplySlurp) (Ability|GameplayCue|MakeGameplayCueParameters :NormalizedMagnitude 0.0 :RawMagnitude 0.0 :Location "0,0,0" :Normal "0,0,0" :Instigator (Variables|Actor|GetInstigator) :EffectCauser self :GameplayEffectLevel 1 :AbilityLevel 1 :bReplicateLocationWhenUsingMinimalRepProxy false)))))

(fn ConstructionScript
  (|CallParentFunction))
